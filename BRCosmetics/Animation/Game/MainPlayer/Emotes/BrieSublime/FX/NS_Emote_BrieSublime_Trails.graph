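# Niagara effect for the BrieSublime emote, drawing swing trails and ground impacts in sync with the animation.
# Changes in User.BurstIndex spawn Arcs and LegArcs meshes, while User.Burst changes spawn Dust, GroundArcs and Rocks, both gated by system age.
# The index becomes the spawn group, which selects count and placement through arrays, masks and switches, with partitions often splitting mirrored pairs.
# Dust tints by time of day; rock visibility varies by platform quality or front-end preview.
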NiagaraSystem /BRCosmetics/Animation/Game/MainPlayer/Emotes/BrieSublime/FX/NS_Emote_BrieSublime_Trails
dependencies {
 usedDataInterfaces [
  &NiagaraDataInterfaceArrayInt32
  &NiagaraDataInterfaceArrayFloat
  &NiagaraDataInterfaceArrayFloat3
  &NiagaraDataInterfaceMeshRendererInfo
  &NiagaraDataInterfaceCurve
  &NiagaraDataInterfaceColorCurve
  &NiagaraDataInterfacePlatformSet
  &NiagaraDataInterfaceSocketReader
 ]
 usedDynamicInputs [&Vector2DFromFloat]
 usedModules [
  &self:SystemSpawnScript.SystemScriptSource.SystemScriptGraph.NiagaraNodeAssignment_0.SetVariables_#1
  &SystemState
 ]
}
emitters {
 Arcs {
  emitterUpdateScript [
   &EmitterState(Life Cycle Mode=::NewEnumerator1, Loop Behavior=::NewEnumerator1, Loop Duration Mode=::NewEnumerator1, UseLoopDelay=false)
   &SpawnBurst_Instantaneous(Spawn Count=2, Use Spawn Probability=false, UseLoopCountLimit=false)
   /Game/Effects/Niagara/Modules/Logic/UpdateValueStateChange(InputIntParam:NiagaraInt32=&MakeIntFromFloat(Float=$User.BurstIndex, Conversion Mode=::NewEnumerator1))
   &self:TriggerSpawn(
    Spawn Count:NiagaraInt32=&SelectIntFromArray(
     Direct Array Index=&MakeIntFromFloat(Float=$User.BurstIndex, Conversion Mode=::NewEnumerator1)
     Int Selection Array=DataInterface() {propertyValues {IntData [2,2,2,2,2,2,2,1,2,2,2,2,1,2]}}
    )
    Spawn Group:NiagaraInt32=&MakeIntFromFloat(Float=$User.BurstIndex, Conversion Mode=::NewEnumerator1)
    Trigger:NiagaraBool=&Boolean_AndOperation(Bool1=$Emitter.UpdateValueStateChange.Changed, Bool2=&SetBoolByFloatComparison(A=$System.Age, B=1.32))
   ) {module=TriggerSpawn}
  ]
  particleSpawnScript [
   &PartitionParticles()
   &InitializeParticle(
    Lifetime=&SelectFloatFromArray(
     Direct Array Index=&SpawnGroup()
     Float Selection Array=DataInterface() {propertyValues {FloatData [0.8,0.6,0.6,0.4,0.6,0.6,0.45,0.6,0.5,0.4,0.5,0.5,0.9,0.6]}}
    )
    Mesh Renderer Array Visibility Mode=::NewEnumerator1
    Mesh Renderer Info=DataInterface() {propertyValues {MeshRenderer=&self:Arcs_0.NiagaraMeshRendererProperties_0}}
    Mesh Renderer Mesh Index:NiagaraInt32=&SelectIntFromArray(
     Direct Array Index=&SpawnGroup()
     Int Selection Array=DataInterface() {propertyValues {IntData [0,1,1,2,1,1,2,1,2,2,2,2,0,0]}}
    )
    Mesh Scale=&Multiply_VectorByFloat(
     Vector=&MaskVector(
      Mask Source=0
      Mask Parameter=?
      Mask=3
      Masked Vector=<1 1 5>
      Passthrough Vector=&MaskVector(
       Mask Source=0
       Mask Parameter=?
       Mask=6
       Masked Vector=<1 1 5>
       Passthrough Vector=&MaskVector(
        Mask Source=0
        Mask Parameter=?
        Mask=8
        Masked Vector=<1 1 5>
        Passthrough Vector=&MaskVector(
         Mask Source=0
         Mask Parameter=?
         Mask=9
         Masked Vector=<1 1 7>
         Passthrough Vector=&MaskVector(
          Mask Source=0
          Mask Parameter=?
          Mask=10
          Masked Vector=<1 1 3>
          Passthrough Vector=&MaskVector(Mask Source=0, Mask Parameter=?, Mask=11, Masked Vector=<1 1 3>, Passthrough Vector=<1 1 1>)
         )
        )
       )
      )
     )
     Float=&SelectFloatFromArray(
      Direct Array Index=&SpawnGroup()
      Float Selection Array=DataInterface() {propertyValues {FloatData [1.5,0.35,0.35,1,0.45,0.45,1,0.5,1,1,1,1,1,1]}}
     )
    )
    Mesh Scale Mode=::NewEnumerator4
    Position Offset=&SelectVectorFromArray(
     Direct Array Index=&SpawnGroup()
     Vector Selection Array=DataInterface() {propertyValues {FloatData [
        <0 0 110>
        <0 15 150>
        <0 15 160>
        <-20 0 270>
        <0 0 95>
        <0 0 95>
        <0 0 270>
        <10 35 130>
        <0 0 260>
        <0 0 -190>
        <0 0 160>
        <0 0 160>
        <0 0 105>
        <0 15 150>
       ]}}
    )
    UsePositionOffset=true
    Write Material Random=true
   )
   &InitialMeshOrientation(
    Mesh Orientation Mode=::NewEnumerator2
    Rotation=&SelectVectorFromArray(
     Direct Array Index=&SpawnGroup()
     Vector Selection Array=DataInterface() {propertyValues {FloatData [
        <0 0.75 0>
        <0 0 0>
        <0 0 0.5>
        <0.5 0 0>
        <0 0.45 0>
        <0 0.45 0>
        <0.5 0 0>
        <0 0.5 0>
        <0.5 0 0>
        <0 0 0>
        <0.5 0 0>
        <0.5 0 0>
        <0 0 -0.2>
        <0 0.75 0>
       ]}}
    )
    Use Rotation Vector=true
   )
   &InitialMeshOrientation(
    Mesh Orientation Mode=::NewEnumerator2
    Rotation=&SelectVectorFromArray(
     Direct Array Index=&SpawnGroup()
     Vector Selection Array=DataInterface() {propertyValues {FloatData [<0.7 0 0>, <0 0 0>, <0 0 0>, <0 0 0>, <0.5 0 0>, <0 0 0.5>, <0 0 0>, <0 0 0.05>, <0 0 0>, <0 0 0>, <0 0 0>, <0 0 0>, <0 0 0>, <0 0 0>]}}
    )
    Use Rotation Vector=true
   ) {module=InitialMeshOrientation001}
   &InitialMeshOrientation(
    Mesh Orientation Mode=::NewEnumerator2
    Rotation=&MaskVector(
     Mask Source=0
     Mask Parameter=?
     Mask=4
     Masked Vector=&MaskVector(Mask Source=1, Mask Parameter=$Particles.PartitionParticles.Partition, Mask=1, Masked Vector=<0 0 0>, Passthrough Vector=<0 0.5 0>)
     Passthrough Vector=&MaskVector(
      Mask Source=0
      Mask Parameter=?
      Mask=5
      Masked Vector=&MaskVector(Mask Source=1, Mask Parameter=$Particles.PartitionParticles.Partition, Mask=1, Masked Vector=<0 0 0>, Passthrough Vector=<0 0 0.5>)
      Passthrough Vector=&MaskVector(
       Mask Source=0
       Mask Parameter=?
       Mask=3
       Masked Vector=&MaskVector(Mask Source=1, Mask Parameter=$Particles.PartitionParticles.Partition, Mask=1, Masked Vector=<0 -0.01 0>, Passthrough Vector=<0 0.01 0>)
       Passthrough Vector=&MaskVector(
        Mask Source=0
        Mask Parameter=?
        Mask=6
        Masked Vector=&MaskVector(Mask Source=1, Mask Parameter=$Particles.PartitionParticles.Partition, Mask=1, Masked Vector=<0 -0.01 0>, Passthrough Vector=<0 0.01 0>)
        Passthrough Vector=&MaskVector(
         Mask Source=0
         Mask Parameter=?
         Mask=8
         Masked Vector=&MaskVector(Mask Source=1, Mask Parameter=$Particles.PartitionParticles.Partition, Mask=1, Masked Vector=<0 -0.01 0>, Passthrough Vector=<0 0.01 0>)
         Passthrough Vector=&MaskVector(
          Mask Source=0
          Mask Parameter=?
          Mask=10
          Masked Vector=&MaskVector(Mask Source=1, Mask Parameter=$Particles.PartitionParticles.Partition, Mask=1, Masked Vector=<0 0 0>, Passthrough Vector=<0.5 0 0>)
          Passthrough Vector=&MaskVector(
           Mask Source=0
           Mask Parameter=?
           Mask=11
           Masked Vector=&MaskVector(Mask Source=1, Mask Parameter=$Particles.PartitionParticles.Partition, Mask=1, Masked Vector=<0.5 0 0>, Passthrough Vector=<0 0 0>)
           Passthrough Vector=<0 0 0>
          )
         )
        )
       )
      )
     )
    )
    Use Rotation Vector=true
   ) {module=InitialMeshOrientation002}
   /SetVariables(Particles.Position=&AddVectorToPosition(
     Position=$Particles.Position
     Vector=&MaskVector(
      Mask Source=0
      Mask Parameter=&MakeIntFromFloat(Float=$User.BurstIndex, Conversion Mode=::NewEnumerator1)
      Mask=0
      Masked Vector=&MaskVector(Mask Source=1, Mask Parameter=$Particles.PartitionParticles.Partition, Mask=1, Masked Vector=<-30 0 0>, Passthrough Vector=<30 0 0>)
      Passthrough Vector=&MaskVector(
       Mask Source=0
       Mask Parameter=&MakeIntFromFloat(Float=$User.BurstIndex, Conversion Mode=::NewEnumerator1)
       Mask=1
       Masked Vector=&MaskVector(Mask Source=1, Mask Parameter=$Particles.PartitionParticles.Partition, Mask=1, Masked Vector=<-30 0 0>, Passthrough Vector=<30 0 0>)
       Passthrough Vector=&MaskVector(
        Mask Source=0
        Mask Parameter=&MakeIntFromFloat(Float=$User.BurstIndex, Conversion Mode=::NewEnumerator1)
        Mask=2
        Masked Vector=&MaskVector(Mask Source=1, Mask Parameter=$Particles.PartitionParticles.Partition, Mask=1, Masked Vector=<-30 0 0>, Passthrough Vector=<30 0 0>)
        Passthrough Vector=&MaskVector(
         Mask Source=0
         Mask Parameter=&MakeIntFromFloat(Float=$User.BurstIndex, Conversion Mode=::NewEnumerator1)
         Mask=3
         Masked Vector=&MaskVector(Mask Source=1, Mask Parameter=$Particles.PartitionParticles.Partition, Mask=1, Masked Vector=<-45 0 0>, Passthrough Vector=<35 0 0>)
         Passthrough Vector=&MaskVector(
          Mask Source=0
          Mask Parameter=&MakeIntFromFloat(Float=$User.BurstIndex, Conversion Mode=::NewEnumerator1)
          Mask=4
          Masked Vector=&MaskVector(Mask Source=1, Mask Parameter=$Particles.PartitionParticles.Partition, Mask=1, Masked Vector=<-10 -25 0>, Passthrough Vector=<10 25 0>)
          Passthrough Vector=&MaskVector(
           Mask Source=0
           Mask Parameter=&MakeIntFromFloat(Float=$User.BurstIndex, Conversion Mode=::NewEnumerator1)
           Mask=5
           Masked Vector=&MaskVector(Mask Source=1, Mask Parameter=$Particles.PartitionParticles.Partition, Mask=1, Masked Vector=<10 -25 0>, Passthrough Vector=<-10 25 0>)
           Passthrough Vector=&MaskVector(
            Mask Source=0
            Mask Parameter=&MakeIntFromFloat(Float=$User.BurstIndex, Conversion Mode=::NewEnumerator1)
            Mask=6
            Masked Vector=&MaskVector(Mask Source=1, Mask Parameter=$Particles.PartitionParticles.Partition, Mask=1, Masked Vector=<-40 0 0>, Passthrough Vector=<40 0 0>)
            Passthrough Vector=&MaskVector(
             Mask Source=0
             Mask Parameter=&MakeIntFromFloat(Float=$User.BurstIndex, Conversion Mode=::NewEnumerator1)
             Mask=7
             Masked Vector=&MaskVector(Mask Source=1, Mask Parameter=$Particles.PartitionParticles.Partition, Mask=1, Masked Vector=<0 0 0>, Passthrough Vector=<0 0 0>)
             Passthrough Vector=&MaskVector(
              Mask Source=0
              Mask Parameter=&MakeIntFromFloat(Float=$User.BurstIndex, Conversion Mode=::NewEnumerator1)
              Mask=8
              Masked Vector=&MaskVector(Mask Source=1, Mask Parameter=$Particles.PartitionParticles.Partition, Mask=1, Masked Vector=<-35 0 0>, Passthrough Vector=<35 0 0>)
              Passthrough Vector=&MaskVector(
               Mask Source=0
               Mask Parameter=&MakeIntFromFloat(Float=$User.BurstIndex, Conversion Mode=::NewEnumerator1)
               Mask=9
               Masked Vector=&MaskVector(Mask Source=1, Mask Parameter=$Particles.PartitionParticles.Partition, Mask=1, Masked Vector=<-30 0 0>, Passthrough Vector=<30 0 0>)
               Passthrough Vector=&MaskVector(
                Mask Source=0
                Mask Parameter=&MakeIntFromFloat(Float=$User.BurstIndex, Conversion Mode=::NewEnumerator1)
                Mask=10
                Masked Vector=&MaskVector(Mask Source=1, Mask Parameter=$Particles.PartitionParticles.Partition, Mask=1, Masked Vector=<-55 0 100>, Passthrough Vector=<55 0 -150>)
                Passthrough Vector=&MaskVector(
                 Mask Source=0
                 Mask Parameter=&MakeIntFromFloat(Float=$User.BurstIndex, Conversion Mode=::NewEnumerator1)
                 Mask=11
                 Masked Vector=&MaskVector(Mask Source=1, Mask Parameter=$Particles.PartitionParticles.Partition, Mask=1, Masked Vector=<-55 0 -150>, Passthrough Vector=<55 0 100>)
                 Passthrough Vector=&MaskVector(
                  Mask Source=0
                  Mask Parameter=&MakeIntFromFloat(Float=$User.BurstIndex, Conversion Mode=::NewEnumerator1)
                  Mask=12
                  Masked Vector=&MaskVector(Mask Source=1, Mask Parameter=$Particles.PartitionParticles.Partition, Mask=1, Masked Vector=<0 0 0>, Passthrough Vector=<0 0 0>)
                  Passthrough Vector=&MaskVector(Mask Source=1, Mask Parameter=$Particles.PartitionParticles.Partition, Mask=1, Masked Vector=<-30 0 0>, Passthrough Vector=<30 0 0>)
                 )
                )
               )
              )
             )
            )
           )
          )
         )
        )
       )
      )
     )
    )) {isSetParameters=true}
  ]
  particleUpdateScript [
   &ParticleState()
   &DynamicMaterialParameters(
    Index 0 Param 1=&FloatFromCurve(FloatCurve=DataInterface() {propertyValues {Curve {
        defaultValue=1
        keys [
         {arriveTangent=-1.1528, interpMode=RCIM_Cubic, leaveTangent=-1.1528, tangentMode=RCTM_User, time=0, value=1}
         {arriveTangent=-8.1952, interpMode=RCIM_Cubic, leaveTangent=-8.1952, tangentMode=RCTM_User, time=1, value=-3.3211}
        ]
       }}})
    Index 0 Param 4=&FloatFromCurve(FloatCurve=DataInterface() {propertyValues {Curve {
        defaultValue=0
        keys [
         {interpMode=RCIM_Cubic, tangentMode=RCTM_User, time=0, value=0}
         {interpMode=RCIM_Cubic, tangentMode=RCTM_User, time=0.5, value=1}
         {tangentMode=RCTM_User, time=1, value=0}
        ]
       }}})
    Module.Index0Float0Write=true
    Module.Index0Float1Write=false
    Module.Index0Float2Write=false
    Module.Index0Float3Write=true
   )
   &UpdateMeshOrientation(
    Rotation Rate=&MaskFloat(
     Mask Source=0
     Mask Parameter=?
     Mask=1
     Masked Float=&FloatFromCurve(
      FloatCurve=DataInterface() {propertyValues {Curve {
         defaultValue=1
         keys [
          {arriveTangent=9.09, interpMode=RCIM_Cubic, leaveTangent=9.09, tangentMode=RCTM_User, time=-0.0849, value=-1}
          {arriveTangent=3.1928, interpMode=RCIM_Cubic, leaveTangent=3.1928, time=0.0855, value=1.0096}
          <at 1 2.4637 RCIM_Cubic>
         ]
        }}}
      Scale Curve=0.3
     )
     Passthrough Float=&MaskFloat(
      Mask Source=0
      Mask Parameter=?
      Mask=2
      Masked Float=&FloatFromCurve(
       FloatCurve=DataInterface() {propertyValues {Curve {
          defaultValue=1
          keys [
           {arriveTangent=9.09, interpMode=RCIM_Cubic, leaveTangent=9.09, tangentMode=RCTM_User, time=-0.0849, value=-1}
           {arriveTangent=3.1928, interpMode=RCIM_Cubic, leaveTangent=3.1928, time=0.0855, value=1.0096}
           <at 1 2.4637 RCIM_Cubic>
          ]
         }}}
       Scale Curve=0.3
      )
      Passthrough Float=&FloatFromCurve(
       FloatCurve=DataInterface() {propertyValues {Curve {
          defaultValue=1
          keys [{arriveTangent=4.6908, interpMode=RCIM_Cubic, leaveTangent=4.6908, tangentMode=RCTM_User, time=0, value=0}, <at 1 3 RCIM_Cubic>]
         }}}
       Scale Curve=&SelectFloatFromArray(
        Direct Array Index=&SpawnGroup()
        Float Selection Array=DataInterface() {propertyValues {FloatData [-0.75,0.3,0.3,0,0.4,0.4,0,0.4,0,0,0,0,-1,-0.4]}}
       )
      )
     )
    )
    Rotation Vector=&SelectVectorFromArray(
     Direct Array Index=&SpawnGroup()
     Vector Selection Array=DataInterface() {propertyValues {FloatData [<0 0 1>, <0 1 0>, <0 1 0>, <0 0 0>, <0 1 0>, <0 1 0>, <0 0 0>, <0 1 0>, <0 0 0>, <0 0 0>, <0 0 0>, <0 0 0>, <0 0 1>, <0 0 1>]}}
    )
   )
   &ScaleMeshSize(Scale Factor=&MaskVector(
     Mask Source=0
     Mask Parameter=?
     Mask=3
     Masked Vector=&MakeVector(
      X=1
      Y=1
      Z=&FloatFromCurve(FloatCurve=DataInterface() {propertyValues {Curve {
          defaultValue=1
          keys [
           {arriveTangent=2.128, interpMode=RCIM_Cubic, leaveTangent=2.128, tangentMode=RCTM_User, time=0, value=0.1}
           {arriveTangent=0.9, interpMode=RCIM_Cubic, leaveTangent=0.9, time=0.2089, value=0.8366}
           {arriveTangent=0.0224, interpMode=RCIM_Cubic, leaveTangent=0.0224, tangentMode=RCTM_User, time=1, value=1}
          ]
         }}})
     )
     Passthrough Vector=&MaskVector(
      Mask Source=0
      Mask Parameter=?
      Mask=6
      Masked Vector=&MakeVector(
       X=1
       Y=1
       Z=&FloatFromCurve(FloatCurve=DataInterface() {propertyValues {Curve {
           defaultValue=1
           keys [
            {arriveTangent=2.128, interpMode=RCIM_Cubic, leaveTangent=2.128, tangentMode=RCTM_User, time=0, value=0.1}
            {arriveTangent=0.9, interpMode=RCIM_Cubic, leaveTangent=0.9, time=0.2058, value=0.8049}
            {arriveTangent=0.0224, interpMode=RCIM_Cubic, leaveTangent=0.0224, tangentMode=RCTM_User, time=1, value=1}
           ]
          }}})
      )
      Passthrough Vector=&MaskVector(
       Mask Source=0
       Mask Parameter=?
       Mask=8
       Masked Vector=&MakeVector(
        X=1
        Y=1
        Z=&FloatFromCurve(FloatCurve=DataInterface() {propertyValues {Curve {
            defaultValue=1
            keys [
             {arriveTangent=2.128, interpMode=RCIM_Cubic, leaveTangent=2.128, tangentMode=RCTM_User, time=0, value=0.1}
             {arriveTangent=0.9, interpMode=RCIM_Cubic, leaveTangent=0.9, time=0.1311, value=0.7672}
             {arriveTangent=0.0224, interpMode=RCIM_Cubic, leaveTangent=0.0224, tangentMode=RCTM_User, time=1, value=1}
            ]
           }}})
       )
       Passthrough Vector=&MaskVector(
        Mask Source=0
        Mask Parameter=?
        Mask=9
        Masked Vector=&MakeVector(
         X=1
         Y=1
         Z=&FloatFromCurve(FloatCurve=DataInterface() {propertyValues {Curve {
             defaultValue=1
             keys [
              {arriveTangent=2.128, interpMode=RCIM_Cubic, leaveTangent=2.128, tangentMode=RCTM_User, time=0, value=0.1}
              {arriveTangent=0.9, interpMode=RCIM_Cubic, leaveTangent=0.9, time=0.1311, value=0.7672}
              {arriveTangent=0.0224, interpMode=RCIM_Cubic, leaveTangent=0.0224, tangentMode=RCTM_User, time=1, value=1}
             ]
            }}})
        )
        Passthrough Vector=&MaskVector(
         Mask Source=0
         Mask Parameter=?
         Mask=10
         Masked Vector=&MakeVector(
          X=1
          Y=1
          Z=&FloatFromCurve(FloatCurve=DataInterface() {propertyValues {Curve {
              defaultValue=1
              keys [
               {arriveTangent=2.128, interpMode=RCIM_Cubic, leaveTangent=2.128, tangentMode=RCTM_User, time=0, value=0.1}
               {arriveTangent=0.9, interpMode=RCIM_Cubic, leaveTangent=0.9, time=0.1311, value=0.7672}
               {arriveTangent=0.0224, interpMode=RCIM_Cubic, leaveTangent=0.0224, tangentMode=RCTM_User, time=1, value=1}
              ]
             }}})
         )
         Passthrough Vector=&MaskVector(
          Mask Source=0
          Mask Parameter=?
          Mask=11
          Masked Vector=&MakeVector(
           X=1
           Y=1
           Z=&FloatFromCurve(FloatCurve=DataInterface() {propertyValues {Curve {
               defaultValue=1
               keys [
                {arriveTangent=2.128, interpMode=RCIM_Cubic, leaveTangent=2.128, tangentMode=RCTM_User, time=0, value=0.1}
                {arriveTangent=0.9, interpMode=RCIM_Cubic, leaveTangent=0.9, time=0.1311, value=0.7672}
                {arriveTangent=0.0224, interpMode=RCIM_Cubic, leaveTangent=0.0224, tangentMode=RCTM_User, time=1, value=1}
               ]
              }}})
          )
          Passthrough Vector=<1 1 1>
         )
        )
       )
      )
     )
    ))
   &ScaleColor(
    Linear Color Curve=DataInterface() {propertyValues {
      AlphaCurve {keys [<at 0 0>,<at 0.2236 1>,<at 0.9811 1>,<at 1 0>]}
      BlueCurve {keys [<at 0 0.4235>,<at 0.4436 0.2784>]}
      GreenCurve {keys [<at 0 0.7884>,<at 0.4436 0.5979>]}
      RedCurve {keys [<at 0 1>,<at 0.4436 1>]}
     }}
    Scale Mode=::NewEnumerator2
   )
  ]
  rendererClasses [&NiagaraMeshRendererProperties]
  renderers [{rendererClass=&NiagaraMeshRendererProperties, rendererIndex=0}]
 }
 Dust {
  emitterUpdateScript [
   &EmitterState(
    Inactive Response=::NewEnumerator1
    Life Cycle Mode=::NewEnumerator1
    Loop Behavior=::NewEnumerator1
    Loop Duration Mode=::NewEnumerator1
    UseLoopDelay=false
   )
   &UpdateValueStateChange(InputIntParam:NiagaraInt32=&MakeIntFromFloat(Float=$User.Burst, Conversion Mode=::NewEnumerator1))
   &self:TriggerSpawn(
    Spawn Count:NiagaraInt32=&SelectIntFromArray(
     Direct Array Index=&MakeIntFromFloat(Float=$User.Burst, Conversion Mode=::NewEnumerator1)
     Int Selection Array=DataInterface() {propertyValues {IntData [17,7,15,15,16,15]}}
    )
    Spawn Group:NiagaraInt32=&MakeIntFromFloat(Float=$User.Burst, Conversion Mode=::NewEnumerator1)
    Trigger:NiagaraBool=&Boolean_AndOperation(Bool1=$Emitter.UpdateValueStateChange.Changed, Bool2=&SetBoolByFloatComparison(A=$System.Age, B=0.57))
   ) {module=TriggerSpawn}
  ]
  particleSpawnScript [
   &PartitionParticles()
   &InitializeParticle(
    Lifetime Max=0.3
    Lifetime Min=0.25
    Lifetime Mode=::NewEnumerator1
    Sprite Size Mode=::NewEnumerator1
    Uniform Sprite Size Max=30
    Uniform Sprite Size Min=20
    UsePositionOffset=false
    Write Material Random=true
   )
   /SetVariables(Particles.OffsetShapePosition:NiagaraPosition=&AddVectorToPosition(
     Position=$Engine.Emitter.SimulationPosition
     Vector=&MergeHelper_VectorSwitch(
      0=&MaskVector(Mask Source=1, Mask Parameter=$Particles.PartitionParticles.Partition, Mask=0, Masked Vector=<-20 0 0>, Passthrough Vector=<25 0 0>)
      1=<-16 -5 0>
      2=&MaskVector(Mask Source=1, Mask Parameter=$Particles.PartitionParticles.Partition, Mask=0, Masked Vector=<-50 10 0>, Passthrough Vector=<40 -15 0>)
      3=&MaskVector(Mask Source=1, Mask Parameter=$Particles.PartitionParticles.Partition, Mask=0, Masked Vector=<40 15 0>, Passthrough Vector=<-50 -10 0>)
      4=&MaskVector(Mask Source=1, Mask Parameter=$Particles.PartitionParticles.Partition, Mask=0, Masked Vector=<-15 0 0>, Passthrough Vector=<18 0 0>)
      5=&MaskVector(Mask Source=1, Mask Parameter=$Particles.PartitionParticles.Partition, Mask=0, Masked Vector=<-20 0 0>, Passthrough Vector=<23 10 0>)
      Index=&SpawnGroup()
      Num=::NewEnumerator4
     )
    )) {isSetParameters=true, module=SetVariables_#2}
   &ShapeLocation(
    Handle Radius=3
    Large Radius=1
    MaskSpawnGroup=false
    Override Seed=false
    Shape Origin=$Particles.OffsetShapePosition
    Shape Primitive=::NewEnumerator3
    UseOwnerScale=true
   )
   &AddVelocity(
    Module.Use Falloff Distance=false
    Override Seed=false
    Velocity Mode=::NewEnumerator1
    Velocity Origin=$Particles.OffsetShapePosition
    Velocity Speed=&RandomRangeFloat(Minimum=300, Maximum=200)
   )
   /SetVariables(Particles.Velocity=&Multiply_Vector(A=$Particles.Velocity, B=<1 1 0.2>)) {isSetParameters=true}
   &CameraOffset(Camera Offset Amount=15)
  ]
  particleUpdateScript [
   &ParticleState()
   &DynamicMaterialParameters(
    Index 0 Param 1=&Add_Float(
     A=&RandomRangeFloat(Minimum=0, Maximum=1)
     B=&FloatFromCurve(
      FloatCurve=DataInterface() {propertyValues {Curve {keys [
          <at 0 0 RCIM_Cubic>
          {arriveTangent=1.5947, interpMode=RCIM_Cubic, leaveTangent=1.5947, tangentMode=RCTM_User, time=0.2557, value=0.5257}
          <at 1 1 RCIM_Cubic>
         ]}}}
      CurveIndex=$Particles.Age
      Scale Curve=0.25
     )
    )
    Module.Index0Float0Write=true
    Module.Index0Float1Write=true
    Module.Index0Float2Write=true
    Module.Index0Float3Write=true
   ) {module=DynamicMaterialParameters001}
   &ScaleColor(
    Scale Alpha=&FloatFromCurve(
     FloatCurve=DataInterface() {propertyValues {Curve {keys [
         {interpMode=RCIM_Cubic, tangentMode=RCTM_User, time=0, value=0}
         {arriveTangentWeight=0.0754, interpMode=RCIM_Cubic, leaveTangentWeight=0.2297, tangentMode=RCTM_User, time=0.1684, value=1}
         {arriveTangent=-0.5014, interpMode=RCIM_Cubic, leaveTangent=-0.5014, tangentMode=RCTM_User, time=0.4045, value=1}
         {arriveTangent=-2.9749, interpMode=RCIM_Cubic, leaveTangent=-2.9749, tangentMode=RCTM_User, time=0.9, value=0}
        ]}}}
     Scale Curve=&FN_FrontEndFloatModifier(Gameplay=0.2, Front End=0.3)
    )
    Scale RGB=&MakeVectorFromLinearColorRGB(Color=&FNTimeOfDayColor(
      Night Color=&RandomRangeLinearColor(Minimum=<rgba 0.4138 0.5201 0.6661 1>, Maximum=<rgba 0.781 0.781 0.781 1>)
      Morning Color=&RandomRangeLinearColor(Minimum=<rgba 0.6882 0.6227 0.9531 1>, Maximum=<rgba 0.6063 0.7457 0.9346 1>)
      Day Color=&RandomRangeLinearColor(Minimum=<rgba 0.478 0.4265 0.4121 0.8>, Maximum=<rgba 0.781 0.718 0.6943 0.9>)
      Evening Color=&RandomRangeLinearColor(Minimum=<rgba 0.478 0.478 0.478 1>, Maximum=<rgba 0.781 0.781 0.781 1>)
     ))
    ScaleA=true
    ScaleRGB=true
   )
   &ScaleSpriteSize(Uniform Curve Sprite Scale=DataInterface() {propertyValues {Curve {keys [
        {arriveTangent=0.8753, interpMode=RCIM_Cubic, leaveTangent=0.8753, tangentMode=RCTM_User, time=0, value=1}
        {arriveTangent=0.0238, interpMode=RCIM_Cubic, leaveTangent=0.0238, tangentMode=RCTM_User, time=1, value=1.3}
       ]}}})
   &Drag(
    Drag=&FloatFromCurve(
     FloatCurve=DataInterface() {propertyValues {Curve {
        defaultValue=1
        keys [{interpMode=RCIM_Cubic, tangentMode=RCTM_User, time=0.0386, value=0}, {interpMode=RCIM_Cubic, tangentMode=RCTM_User, time=0.1066, value=1}]
       }}}
     Scale Curve=&RandomRangeFloat(Minimum=9, Maximum=7)
    )
    Use Linear Drag=true
    Use Rotational Drag=true
   )
   &SolveForcesAndVelocity(Clamp Velocity=false, Limit Acceleration=false)
  ]
  rendererClasses [&NiagaraSpriteRendererProperties]
  renderers [
   {rendererClass=&NiagaraSpriteRendererProperties}
   {rendererClass=&NiagaraSpriteRendererProperties, rendererIndex=1}
  ]
 }
 GroundArcs {
  emitterUpdateScript [
   &EmitterState(Life Cycle Mode=::NewEnumerator1, Loop Behavior=::NewEnumerator1, Loop Duration Mode=::NewEnumerator1, UseLoopDelay=false)
   &UpdateValueStateChange(InputIntParam:NiagaraInt32=&MakeIntFromFloat(Float=$User.Burst, Conversion Mode=::NewEnumerator1))
   &self:TriggerSpawn(
    Spawn Count=2
    Spawn Group:NiagaraInt32=&MakeIntFromFloat(Float=$User.Burst, Conversion Mode=::NewEnumerator1)
    Trigger:NiagaraBool=&Boolean_AndOperation(Bool1=$Emitter.UpdateValueStateChange.Changed, Bool2=&SetBoolByFloatComparison(A=$System.Age, B=0.57))
   ) {module=TriggerSpawn}
  ]
  particleSpawnScript [
   &PartitionParticles()
   &InitializeParticle(
    Color=&FNTimeOfDayColor(Night Color=<rgba 1 1 1 0.2>, Morning Color=<rgba 1 1 1 0.2>, Day Color=<rgba 1 1 1 0.2>, Evening Color=<rgba 1 1 1 0.2>)
    Lifetime=0.5
    Mesh Scale Mode=::NewEnumerator3
    Mesh Uniform Scale=0.4
    UsePositionOffset=false
    Write Material Random=true
   )
   &InitialMeshOrientation(
    Mesh Orientation Mode=::NewEnumerator2
    Orientation Vector=<0 1 0>
    Rotation=&MaskVector(Mask Source=1, Mask Parameter=$Particles.PartitionParticles.Partition, Mask=0, Masked Vector=<0 0 0>, Passthrough Vector=<0 0 0.5>)
    Use Rotation Vector=true
   )
   /SetVariables(Particles.Position=&AddVectorToPosition(
     Position=$Particles.Position
     Vector=&MergeHelper_VectorSwitch(
      0=&MaskVector(Mask Source=1, Mask Parameter=$Particles.PartitionParticles.Partition, Mask=1, Masked Vector=<-50 0 0>, Passthrough Vector=<50 0 0>)
      1=&MaskVector(Mask Source=1, Mask Parameter=$Particles.PartitionParticles.Partition, Mask=1, Masked Vector=<-60 0 0>, Passthrough Vector=<10 0 0>)
      2=&MaskVector(Mask Source=1, Mask Parameter=$Particles.PartitionParticles.Partition, Mask=1, Masked Vector=<-70 0 0>, Passthrough Vector=<70 0 0>)
      3=&MaskVector(Mask Source=1, Mask Parameter=$Particles.PartitionParticles.Partition, Mask=0, Masked Vector=<70 0 0>, Passthrough Vector=<-70 0 0>)
      4=&MaskVector(Mask Source=1, Mask Parameter=$Particles.PartitionParticles.Partition, Mask=1, Masked Vector=<-50 0 0>, Passthrough Vector=<50 0 0>)
      5=&MaskVector(Mask Source=1, Mask Parameter=$Particles.PartitionParticles.Partition, Mask=1, Masked Vector=<-50 0 0>, Passthrough Vector=<50 0 0>)
      Index=&SpawnGroup()
      Num=::NewEnumerator4
     )
    )) {isSetParameters=true}
   &CameraOffset(Camera Offset Amount=20)
  ]
  particleUpdateScript [
   &ParticleState()
   &ScaleMeshSize(Scale Factor=&VectorFromFloat(Value=&FloatFromCurve(FloatCurve=DataInterface() {propertyValues {Curve {
         defaultValue=1
         keys [
          {tangentMode=RCTM_User, time=0, value=1}
          {arriveTangent=-0.0405, interpMode=RCIM_Cubic, leaveTangent=-0.0405, tangentMode=RCTM_User, time=0.8283, value=1}
          {arriveTangent=-8.0089, interpMode=RCIM_Cubic, leaveTangent=-8.0089, tangentMode=RCTM_User, time=1, value=0}
         ]
        }}})))
   &DynamicMaterialParameters(
    Index 0 Param 2=&FloatFromCurve(FloatCurve=DataInterface() {propertyValues {Curve {
        defaultValue=1
        keys [{interpMode=RCIM_Cubic, tangentMode=RCTM_User, time=0, value=-0.4}, {interpMode=RCIM_Cubic, tangentMode=RCTM_User, time=1, value=1}]
       }}})
    Module.Index0Float0Write=true
    Module.Index0Float1Write=true
    Module.Index0Float2Write=true
    Module.Index0Float3Write=true
   )
  ]
  rendererClasses [&NiagaraMeshRendererProperties]
  renderers [{rendererClass=&NiagaraMeshRendererProperties, rendererIndex=0}]
 }
 LegArcs {
  emitterUpdateScript [
   &EmitterState(Life Cycle Mode=::NewEnumerator1, Loop Behavior=::NewEnumerator1, Loop Duration Mode=::NewEnumerator1, UseLoopDelay=false)
   &UpdateValueStateChange(InputIntParam:NiagaraInt32=&MakeIntFromFloat(Float=$User.BurstIndex, Conversion Mode=::NewEnumerator1))
   &self:TriggerSpawn(
    Spawn Count:NiagaraInt32=&SelectIntFromArray(
     Direct Array Index=&MakeIntFromFloat(Float=$User.BurstIndex, Conversion Mode=::NewEnumerator1)
     Int Selection Array=DataInterface() {propertyValues {IntData [0,2,2,0,2,2,0,1,0,0,2,2,0,0]}}
    )
    Spawn Group:NiagaraInt32=&MakeIntFromFloat(Float=$User.BurstIndex, Conversion Mode=::NewEnumerator1)
    Trigger:NiagaraBool=&Boolean_AndOperation(Bool1=$Emitter.UpdateValueStateChange.Changed, Bool2=&SetBoolByFloatComparison(A=$System.Age, B=1.32))
   ) {module=TriggerSpawn}
  ]
  particleSpawnScript [
   &PartitionParticles()
   &InitializeParticle(
    Color=<rgba 1 1 1 0.85>
    Lifetime=0.5
    Mesh Renderer Array Visibility Mode=::NewEnumerator1
    Mesh Renderer Info=DataInterface() {propertyValues {MeshRenderer=&self:LegArcs_0.NiagaraMeshRendererProperties_0}}
    Mesh Renderer Mesh Index:NiagaraInt32=&SelectIntFromArray(
     Direct Array Index=&SpawnGroup()
     Int Selection Array=DataInterface() {propertyValues {IntData [2,1,1,2,1,1,2,0,2,2,1,1,2,2]}}
    )
    Mesh Scale Mode=::NewEnumerator3
    Mesh Uniform Scale=&SelectFloatFromArray(
     Direct Array Index=&SpawnGroup()
     Float Selection Array=DataInterface() {propertyValues {FloatData [1,0.3,0.3,1,0.6,0.6,1,1,1,1,0.2,0.2,1,1]}}
    )
    Position Offset=&SelectVectorFromArray(
     Direct Array Index=&SpawnGroup()
     Vector Selection Array=DataInterface() {propertyValues {FloatData [<0 0 0>, <0 0 0>, <0 0 0>, <0 0 0>, <0 -30 -15>, <0 30 -15>, <0 0 0>, <50 0 5>, <0 0 0>, <0 0 0>, <0 0 10>, <0 0 10>, <0 0 0>, <0 0 0>]}}
    )
    UsePositionOffset=true
    Write Material Random=true
   )
   &InitialMeshOrientation(
    Mesh Orientation Mode=::NewEnumerator2
    Rotation=&SelectVectorFromArray(
     Direct Array Index=&SpawnGroup()
     Vector Selection Array=DataInterface() {propertyValues {FloatData [
        <0 0 0>
        <0.25 0 0>
        <0.75 0 0>
        <0 0 0>
        <0 -0.2 0>
        <0 -0.2 0>
        <0 0 0>
        <0 0 0>
        <0 0 0>
        <0 0 0>
        <0 0.75 0>
        <0 0.75 0>
        <0 0 0>
        <0 0 0>
       ]}}
    )
    Use Rotation Vector=true
   )
   &InitialMeshOrientation(
    Mesh Orientation Mode=::NewEnumerator2
    Rotation=&SelectVectorFromArray(
     Direct Array Index=&SpawnGroup()
     Vector Selection Array=DataInterface() {propertyValues {FloatData [<0 0 0>, <0 0 0.1>, <0 0 0.4>, <0 0 0>, <0 0 0>, <0 0 0>, <0 0 0>, <0 0 0>, <0 0 0>, <0 0 0>, <0 -0.2 0>, <0 -0.2 0>, <0 0 0>, <0 0 0>]}}
    )
    Use Rotation Vector=true
   ) {module=InitialMeshOrientation001}
   &InitialMeshOrientation(
    Mesh Orientation Mode=::NewEnumerator2
    Rotation=&MaskVector(
     Mask Source=0
     Mask Parameter=?
     Mask=1
     Masked Vector=&MaskVector(Mask Source=1, Mask Parameter=$Particles.PartitionParticles.Partition, Mask=1, Masked Vector=<0 0 0>, Passthrough Vector=<0 0 0.5>)
     Passthrough Vector=&MaskVector(
      Mask Source=0
      Mask Parameter=?
      Mask=2
      Masked Vector=&MaskVector(Mask Source=1, Mask Parameter=$Particles.PartitionParticles.Partition, Mask=1, Masked Vector=<0 0 0.5>, Passthrough Vector=<0 0 0>)
      Passthrough Vector=&MaskVector(
       Mask Source=0
       Mask Parameter=?
       Mask=10
       Masked Vector=&MaskVector(Mask Source=1, Mask Parameter=$Particles.PartitionParticles.Partition, Mask=1, Masked Vector=<0 0 0.5>, Passthrough Vector=<0 0 0>)
       Passthrough Vector=&MaskVector(
        Mask Source=0
        Mask Parameter=?
        Mask=11
        Masked Vector=&MaskVector(Mask Source=1, Mask Parameter=$Particles.PartitionParticles.Partition, Mask=1, Masked Vector=<0 0 0.5>, Passthrough Vector=<0 0 0>)
        Passthrough Vector=<0 0 0>
       )
      )
     )
    )
    Use Rotation Vector=true
   ) {module=InitialMeshOrientation002}
   /SetVariables(Particles.Position=&AddVectorToPosition(
     Position=$Particles.Position
     Vector=&MaskVector(
      Mask Source=0
      Mask Parameter=&MakeIntFromFloat(Float=$User.BurstIndex, Conversion Mode=::NewEnumerator1)
      Mask=1
      Masked Vector=&MaskVector(Mask Source=1, Mask Parameter=$Particles.PartitionParticles.Partition, Mask=1, Masked Vector=<30 0 0>, Passthrough Vector=<-30 0 0>)
      Passthrough Vector=&MaskVector(
       Mask Source=0
       Mask Parameter=&MakeIntFromFloat(Float=$User.BurstIndex, Conversion Mode=::NewEnumerator1)
       Mask=2
       Masked Vector=&MaskVector(Mask Source=1, Mask Parameter=$Particles.PartitionParticles.Partition, Mask=1, Masked Vector=<30 0 0>, Passthrough Vector=<-30 0 0>)
       Passthrough Vector=&MaskVector(
        Mask Source=0
        Mask Parameter=&MakeIntFromFloat(Float=$User.BurstIndex, Conversion Mode=::NewEnumerator1)
        Mask=10
        Masked Vector=&MaskVector(Mask Source=1, Mask Parameter=$Particles.PartitionParticles.Partition, Mask=1, Masked Vector=<60 0 0>, Passthrough Vector=<-60 0 0>)
        Passthrough Vector=&MaskVector(
         Mask Source=0
         Mask Parameter=&MakeIntFromFloat(Float=$User.BurstIndex, Conversion Mode=::NewEnumerator1)
         Mask=11
         Masked Vector=&MaskVector(Mask Source=1, Mask Parameter=$Particles.PartitionParticles.Partition, Mask=1, Masked Vector=<60 0 0>, Passthrough Vector=<-60 0 0>)
         Passthrough Vector=<0 0 0>
        )
       )
      )
     )
    )) {isSetParameters=true}
   &CameraOffset(Camera Offset Amount=20)
  ]
  particleUpdateScript [
   &ParticleState()
   &DynamicMaterialParameters(
    Index 0 Param 1=&FloatFromCurve(FloatCurve=DataInterface() {propertyValues {Curve {
        defaultValue=1
        keys [
         {arriveTangent=-1.1528, interpMode=RCIM_Cubic, leaveTangent=-1.1528, tangentMode=RCTM_User, time=0, value=1}
         {arriveTangent=-8.1952, interpMode=RCIM_Cubic, leaveTangent=-8.1952, tangentMode=RCTM_User, time=1, value=-3.3211}
        ]
       }}})
    Index 0 Param 4=&FloatFromCurve(FloatCurve=DataInterface() {propertyValues {Curve {
        defaultValue=0
        keys [
         {interpMode=RCIM_Cubic, tangentMode=RCTM_User, time=0, value=0}
         {interpMode=RCIM_Cubic, tangentMode=RCTM_User, time=0.5, value=1}
         {tangentMode=RCTM_User, time=1, value=0}
        ]
       }}})
    Module.Index0Float0Write=true
    Module.Index0Float1Write=false
    Module.Index0Float2Write=false
    Module.Index0Float3Write=true
   )
   &UpdateMeshOrientation(
    Rotation Rate=&FloatFromCurve(
     FloatCurve=DataInterface() {propertyValues {Curve {
        defaultValue=1
        keys [{arriveTangent=4.6908, interpMode=RCIM_Cubic, leaveTangent=4.6908, tangentMode=RCTM_User, time=0, value=0}, <at 1 3 RCIM_Cubic>]
       }}}
     Scale Curve=&SelectFloatFromArray(
      Direct Array Index=&SpawnGroup()
      Float Selection Array=DataInterface() {propertyValues {FloatData [0,0.3,0.3,0,0.4,0.4,0,0,0,0,0.3,0.3,0,0]}}
     )
    )
    Rotation Vector=&SelectVectorFromArray(
     Direct Array Index=&SpawnGroup()
     Vector Selection Array=DataInterface() {propertyValues {FloatData [<0 0 0>, <0 1 0>, <0 1 0>, <0 0 0>, <0 1 0>, <0 1 0>, <0 0 0>, <0 0 0>, <0 0 0>, <0 0 0>, <0 1 0>, <0 1 0>, <0 0 0>, <0 0 0>]}}
    )
   )
   &ScaleColor(
    Linear Color Curve=DataInterface() {propertyValues {
      AlphaCurve {keys [<at 0 0>,<at 0.2236 1>,<at 0.9811 1>,<at 1 0>]}
      BlueCurve {keys [<at 0 0.4235>,<at 0.4436 0.2784>]}
      GreenCurve {keys [<at 0 0.7884>,<at 0.4436 0.5979>]}
      RedCurve {keys [<at 0 1>,<at 0.4436 1>]}
     }}
    Scale Mode=::NewEnumerator2
   )
   &ScaleMeshSize(Scale Factor=&MaskVector(
     Mask Source=0
     Mask Parameter=?
     Mask=7
     Masked Vector=&MakeVector(
      X=&FloatFromCurve(FloatCurve=DataInterface() {propertyValues {Curve {
          defaultValue=0
          keys [
           {arriveTangent=4.1785, interpMode=RCIM_Cubic, leaveTangent=4.1785, tangentMode=RCTM_User, time=0, value=0}
           {arriveTangent=1.1179, interpMode=RCIM_Cubic, leaveTangent=1.1179, tangentMode=RCTM_User, time=0.1436, value=0.7274}
           {arriveTangent=0.0224, interpMode=RCIM_Cubic, leaveTangent=0.0224, tangentMode=RCTM_User, time=0.5043, value=1}
          ]
         }}})
      Y=1
      Z=1
     )
     Passthrough Vector=<1 1 1>
    ))
  ]
  rendererClasses [&NiagaraMeshRendererProperties]
  renderers [{rendererClass=&NiagaraMeshRendererProperties, rendererIndex=0}]
 }
 Rocks {
  emitterUpdateScript [
   &EmitterState(
    Inactive Response=::NewEnumerator1
    Life Cycle Mode=::NewEnumerator1
    Loop Behavior=::NewEnumerator1
    Loop Duration Mode=::NewEnumerator1
    UseLoopDelay=false
   )
   &UpdateValueStateChange(InputIntParam:NiagaraInt32=&MakeIntFromFloat(Float=$User.Burst, Conversion Mode=::NewEnumerator1))
   &self:TriggerSpawn(
    Spawn Count:NiagaraInt32=&SelectIntFromArray(
     Direct Array Index=&MakeIntFromFloat(Float=$User.Burst, Conversion Mode=::NewEnumerator1)
     Int Selection Array=DataInterface() {propertyValues {IntData [17,7,15,15,16,15]}}
    )
    Spawn Group:NiagaraInt32=&MakeIntFromFloat(Float=$User.Burst, Conversion Mode=::NewEnumerator1)
    Trigger:NiagaraBool=&Boolean_AndOperation(Bool1=$Emitter.UpdateValueStateChange.Changed, Bool2=&SetBoolByFloatComparison(A=$System.Age, B=0.57))
   ) {module=TriggerSpawn}
  ]
  particleSpawnScript [
   /SetVariables(Particles.VisibilityTag=&MakeCustomIntFromBool(
     Bool=&Boolean_OrOperation(
      Bool1=&IsPlatformSetActive(New Platform Set=DataInterface() {propertyValues {Platforms {qualityLevelMask=-2}}})
      Bool2=$User.bIsFrontEndPreview
     )
     True INT=&SelectIntFromArray(
      Array Sampling Mode=::NewEnumerator1
      Direct Array Index=?
      Int Selection Array=DataInterface() {propertyValues {IntData [0,1,2,0,0,2]}}
      Random Seed=0
     )
     False INT=0
    )) {isSetParameters=true}
   &PartitionParticles()
   &InitializeParticle(
    Lifetime Max=0.4
    Lifetime Min=0.7
    Lifetime Mode=::NewEnumerator1
    Mesh Scale Mode=::NewEnumerator1
    Mesh Uniform Scale Max=0.4
    Mesh Uniform Scale Min=0.3
    Sprite Rotation Mode=::NewEnumerator2
    Sprite Size Mode=::NewEnumerator1
    Uniform Sprite Size Min=15
    UsePositionOffset=false
    Write Material Random=true
   )
   /SetVariables(Particles.OffsetShapePosition:NiagaraPosition=&AddVectorToPosition(
     Position=$Engine.Emitter.SimulationPosition
     Vector=&MergeHelper_VectorSwitch(
      0=&MaskVector(Mask Source=1, Mask Parameter=$Particles.PartitionParticles.Partition, Mask=0, Masked Vector=<-20 0 5>, Passthrough Vector=<25 0 5>)
      1=<-16 -5 5>
      2=&MaskVector(Mask Source=1, Mask Parameter=$Particles.PartitionParticles.Partition, Mask=0, Masked Vector=<-50 10 5>, Passthrough Vector=<40 -15 5>)
      3=&MaskVector(Mask Source=1, Mask Parameter=$Particles.PartitionParticles.Partition, Mask=0, Masked Vector=<40 15 5>, Passthrough Vector=<-50 -10 5>)
      4=&MaskVector(Mask Source=1, Mask Parameter=$Particles.PartitionParticles.Partition, Mask=0, Masked Vector=<-15 0 5>, Passthrough Vector=<18 0 5>)
      5=&MaskVector(Mask Source=1, Mask Parameter=$Particles.PartitionParticles.Partition, Mask=0, Masked Vector=<-20 0 5>, Passthrough Vector=<23 10 5>)
      Index=&SpawnGroup()
      Num=::NewEnumerator4
     )
    )) {isSetParameters=true, module=SetVariables_#2}
   &ShapeLocation(
    Handle Radius=3
    Large Radius=2
    MaskSpawnGroup=false
    Override Seed=false
    Shape Origin=$Particles.OffsetShapePosition
    Shape Primitive=::NewEnumerator3
    UseOwnerScale=true
   )
   &AddVelocity(
    Module.Use Falloff Distance=false
    Origin Offset=<0 0 -2.5>
    Override Seed=false
    Velocity Mode=::NewEnumerator1
    Velocity Origin=$Particles.OffsetShapePosition
    Velocity Speed=&RandomRangeFloat(Minimum=300, Maximum=200)
   )
   &CameraOffset(Camera Offset Amount=15)
   &InitialMeshOrientation(Use Rotation Vector=false)
  ]
  particleUpdateScript [
   &ParticleState()
   &GravityForce()
   &Collision(
    Age Colliding Particles=false
    Collision Enabled=&IsPlatformSetActive(New Platform Set=DataInterface() {propertyValues {Platforms {qualityLevelMask=-8}}})
    KillOnCollision=false
    Module.Control Rest State Via Surface Normal=false
    Module.EnableMaxCollisionCount=false
    Module.Randomize Collision Normal Vector=false
    Radius Calculation Type=::NewEnumerator1
   )
   &Collision(
    Age Colliding Particles=false
    Analytical Collision Plane Position 1:Vector3f=<0 0 -20>
    Analytical Collision Plane Space:ENiagaraCoordinateSpace=::Local
    CPU Collision Type=::NewEnumerator3
    Collision Enabled=&IsPlatformSetActive(New Platform Set=DataInterface() {propertyValues {Platforms {qualityLevelMask=-28}}})
    KillOnCollision=false
    Module.Control Rest State Via Surface Normal=false
    Module.EnableMaxCollisionCount=false
    Module.Randomize Collision Normal Vector=false
    Perform 2 Analytical Collision Queries=false
   ) {module=Collision001}
   &ScaleColor(
    Scale Alpha=&FloatFromCurve(
     FloatCurve=DataInterface() {propertyValues {Curve {keys [
         {interpMode=RCIM_Cubic, tangentMode=RCTM_User, time=0, value=0}
         {arriveTangentWeight=0.0754, interpMode=RCIM_Cubic, leaveTangentWeight=0.2297, tangentMode=RCTM_User, time=0.1684, value=1}
         {arriveTangent=-0.5014, interpMode=RCIM_Cubic, leaveTangent=-0.5014, tangentMode=RCTM_User, time=0.4045, value=1}
         {arriveTangent=-2.9749, interpMode=RCIM_Cubic, leaveTangent=-2.9749, tangentMode=RCTM_User, time=0.9, value=0}
        ]}}}
     Scale Curve=&FN_FrontEndFloatModifier(Gameplay=0.3, Front End=0.4)
    )
    Scale RGB=&MakeVectorFromLinearColorRGB(Color=&FNTimeOfDayColor(
      Night Color=&RandomRangeLinearColor(Minimum=<rgba 0.4138 0.5201 0.6661 1>, Maximum=<rgba 0.781 0.781 0.781 1>)
      Morning Color=&RandomRangeLinearColor(Minimum=<rgba 0.6882 0.6227 0.9531 1>, Maximum=<rgba 0.6063 0.7457 0.9346 1>)
      Day Color=&RandomRangeLinearColor(Minimum=<rgba 0.478 0.4265 0.4121 0.8>, Maximum=<rgba 0.781 0.718 0.6943 0.9>)
      Evening Color=&RandomRangeLinearColor(Minimum=<rgba 0.478 0.478 0.478 1>, Maximum=<rgba 0.781 0.781 0.781 1>)
     ))
    ScaleA=true
    ScaleRGB=true
   )
   &ScaleSpriteSize(Uniform Curve Sprite Scale=DataInterface() {propertyValues {Curve {
       defaultValue=1
       keys [
        {tangentMode=RCTM_User, time=0, value=1}
        {arriveTangent=-0.0405, interpMode=RCIM_Cubic, leaveTangent=-0.0405, tangentMode=RCTM_User, time=0.7009, value=1}
        {arriveTangent=-8.0089, interpMode=RCIM_Cubic, leaveTangent=-8.0089, tangentMode=RCTM_User, time=1, value=0}
       ]
      }}})
   &ScaleMeshSize(Scale Factor=&VectorFromFloat(Value=&FloatFromCurve(FloatCurve=DataInterface() {propertyValues {Curve {
         defaultValue=1
         keys [
          {tangentMode=RCTM_User, time=0, value=1}
          {arriveTangent=-0.0405, interpMode=RCIM_Cubic, leaveTangent=-0.0405, tangentMode=RCTM_User, time=0.1895, value=1}
          {arriveTangent=-1.0531, interpMode=RCIM_Cubic, leaveTangent=-1.0531, tangentMode=RCTM_User, time=1, value=0}
         ]
        }}})))
   &Drag(
    Drag=&FloatFromCurve(
     FloatCurve=DataInterface() {propertyValues {Curve {
        defaultValue=1
        keys [{interpMode=RCIM_Cubic, tangentMode=RCTM_User, time=0.0386, value=0}, {interpMode=RCIM_Cubic, tangentMode=RCTM_User, time=0.0735, value=1}]
       }}}
     Scale Curve=&RandomRangeFloat(Minimum=4, Maximum=5)
    )
    Use Linear Drag=true
    Use Rotational Drag=true
   )
   &SolveForcesAndVelocity(Clamp Velocity=false, Limit Acceleration=false)
   &KillParticlesInVolume(Sphere Radius=5)
  ]
  rendererClasses [&NiagaraMeshRendererProperties]
  renderers [
   {rendererClass=&NiagaraMeshRendererProperties, rendererIndex=0}
   {rendererClass=&NiagaraMeshRendererProperties, rendererIndex=1}
  ]
 }
}
summary {
 systemName=NS_Emote_BrieSublime_Trails
 emitters [
  {emitterName=Arcs, rendererClasses [&NiagaraMeshRendererProperties]}
  {emitterName=Dust, rendererClasses [&NiagaraSpriteRendererProperties]}
  {emitterName=Rocks, rendererClasses [&NiagaraMeshRendererProperties]}
  {emitterName=LegArcs, rendererClasses [&NiagaraMeshRendererProperties]}
  {emitterName=GroundArcs, rendererClasses [&NiagaraMeshRendererProperties]}
 ]
 userVariables [
  User.bIsFrontEnd {type=NiagaraBool}
  User.bIsFrontEndPreview {type=NiagaraBool}
  User.Burst {type=NiagaraFloat, defaultValue=0, description=""}
  User.BurstIndex {type=NiagaraFloat, defaultValue=0, description=""}
 ]
}
